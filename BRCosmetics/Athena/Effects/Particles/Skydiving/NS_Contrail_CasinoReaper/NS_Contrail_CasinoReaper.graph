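# Niagara skydiving contrail for the CasinoReaper cosmetic, trailing card meshes, flares and speed lines from the character.
# Card_Base, Long_Cards and Flare partition particles across sockets, adding mirrored sine/cosine offsets for spiral motion.
# Long_Cards widens its spiral with age, orienting along it and pulsing mesh scale with camera fade.
# Speed_Lines are velocity-stretched sprites using dive and time-of-day alpha in gameplay, with separate front-end values.
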NiagaraSystem /BRCosmetics/Athena/Effects/Particles/Skydiving/NS_Contrail_CasinoReaper/NS_Contrail_CasinoReaper
dependencies {
 usedDataInterfaces [
  &NiagaraDataInterfaceArrayFloat2
  &NiagaraDataInterfaceMeshRendererInfo
  &NiagaraDataInterfaceCurve
  &NiagaraDataInterfaceCamera
  &NiagaraDataInterfaceVectorCurve
  &NiagaraDataInterfaceColorCurve
  &NiagaraDataInterfaceVector2DCurve
  &NiagaraDataInterfaceSocketReader
  &NiagaraDataInterfaceSkeletalMesh
 ]
 usedDynamicInputs []
 usedModules [
  &self:SystemSpawnScript.SystemScriptSource.SystemScriptGraph.NiagaraNodeAssignment_0.SetVariables_#1
  &self:SystemSpawnScript.SystemScriptSource.SystemScriptGraph.NiagaraNodeAssignment_3.SetVariables_#2
  &self:SystemSpawnScript.SystemScriptSource.SystemScriptGraph.NiagaraNodeAssignment_4.SetVariables_#3
  &SystemState
  /Game/Effects/Niagara/Modules/Skydiving/FN_DivingAttributes/v5/FN_DivingAttributes
  &self:SystemSpawnScript.SystemScriptSource.SystemScriptGraph.NiagaraNodeAssignment_2.SetVariables_#4
 ]
}
emitters {
 Card_Base {
  emitterUpdateScript [
   &EmitterState(
    Delay First Loop Only=true
    Inactive Response=::NewEnumerator1
    Life Cycle Mode=::NewEnumerator1
    Loop Delay=0.5
    Loop Duration=1
    UseLoopDelay=true
   )
   &SpawnRate(SpawnRate=50, Use Spawn Probability=true)
  ]
  particleSpawnScript [
   &PartitionParticles(Write Partition to Ribbon ID=::NewEnumerator0)
   &InitializeParticle(
    Color=<rgba 3 1.5567 1.5 1>
    Lifetime=&RandomRangeFloat(Minimum=0.5, Maximum=1.5)
    Mesh Renderer Array Visibility Mode=::NewEnumerator1
    Mesh Renderer Info=DataInterface() {propertyValues {MeshRenderer=&self:Card_Base_0.NiagaraMeshRendererProperties_0}}
    Mesh Renderer Mesh Index:NiagaraInt32=/Niagara/DynamicInputs/Divide/Divide_Int(A:NiagaraInt32=&ReturnParticleIDAttributes(), B:NiagaraInt32=2)
    Mesh Scale=<0.2 0.2 0.01>
    Mesh Scale Mode=::NewEnumerator4
    Sprite Size=&SelectVector2DFromArray(
     Direct Array Index=$Particles.PartitionParticles.Partition
     Vector 2D Selection Array=DataInterface() {propertyValues {FloatData [<60 230>, <60 230>, <30 100>, <30 100>]}}
    )
    Sprite Size Mode=::NewEnumerator4
    UsePositionOffset=false
    Write Material Random=true
   )
   &InitialMeshOrientation(Mesh Orientation Mode=::NewEnumerator7, Use Rotation Vector=true)
   &CameraOffset(Camera Offset Amount=&Add_Float(
     A=0
     B=&Multiply_FloatByInt(Float=-0.1, Integer=&ModuloInt(A=&ReturnParticleIDAttributes(), B=60))
    )) {enabled=false}
  ]
  particleUpdateScript [
   &ParticleState()
   &ScaleMeshSize(Scale Factor=&VectorFromFloat(Value=&FloatFromCurve(FloatCurve=DataInterface() {propertyValues {Curve {keys [<at 0.1 0>,<at 0.2 1>,<at 0.8 1>,<at 1 0>]}}}))) {enabled=false}
   &ScaleMeshSize(Scale Factor=&VectorFromCurve(VectorCurve=DataInterface() {propertyValues {
       XCurve {keys [<at 0.09 0>,<at 0.1 1>,<at 1 1>]}
       YCurve {keys [<at 0 0>,<at 0.2 1>,<at 0.9 1>,<at 1 0>]}
       ZCurve {keys [<at 0.09 0>,<at 0.1 1>,<at 1 1>]}
      }})) {module=ScaleMeshSize001}
   &SocketLocation(
    Interpolate Socket Sampling=true
    Orientation Sampling=::NewEnumerator1
    SamplePosOffset=true
    Sampled Position Offset=&MergeHelper_VectorSwitch(0=<5 -7 0>, 1=<-5 7 0>, 2=?, 3=?, Index=$Particles.PartitionParticles.Partition, Num=::NewEnumerator0)
    Socket Index=$Particles.PartitionParticles.Partition
    Socket Reader=$System.Socket Reader
    Socket Sampling Mode=::NewEnumerator3
    VelocitySpeedLimit=false
   )
   &SkeletalMeshLocation(
    Bone / Socket Index=$Particles.PartitionParticles.Partition
    Bone Sampling Mode=::NewEnumerator4
    MaskSpawnGroup=false
    Orientation Sampling=::NewEnumerator1
    Position Sampling=::NewEnumerator2
    Sampled Position Offset=&MergeHelper_VectorSwitch(0=<3 0 0>, 1=<-3 0 0>, 2=?, 3=?, Index=$Particles.PartitionParticles.Partition, Num=::NewEnumerator0)
    Skeletal Mesh=$System.Skeletal Mesh
   ) {enabled=false}
   /SetVariables(Particles.CustomRandom:NiagaraFloat=&Add_Float(
     A=&RandomRangeFloat(
      Minimum=0
      Maximum=50
      Randomness Mode=::NewEnumerator1
      Random Seed=&Add_Integer(A=$Particles.PartitionParticles.Partition, B=0)
      Fixed Random Seed=true
     )
     B=&MergeHelper_FloatSwitch(0=0.7, 1=0.88, 2=?, 3=?, Index=$Particles.PartitionParticles.Partition, Num=::NewEnumerator0)
    )) {isSetParameters=true, module=SetVariables_#5}
   /SetVariables(Particles.SpiralVector:Vector3f=&MakeVector(
     X=&Waveform(
      Global Input Position=$Particles.CustomRandom
      Global Amplitude Scale=&MergeHelper_FloatSwitch(0=-1, 1=1, 2=?, 3=?, Index=$Particles.PartitionParticles.Partition, Num=::NewEnumerator0)
      "Waveform [1]"=::NewEnumerator0
      "[1] Phase"=$Particles.Age
      "[1] Frequency"=1.5
      "[1] Amplitude Min/Max"=<-1 1>
      "[1] Scale"=&RandomRangeFloat(Minimum=10, Maximum=12)
     )
     Y=&Waveform(
      Global Input Position=$Particles.CustomRandom
      "Waveform [1]"=::NewEnumerator1
      "[1] Phase"=$Particles.Age
      "[1] Frequency"=1.5
      "[1] Amplitude Min/Max"=<-1 1>
      "[1] Scale"=&RandomRangeFloat(Minimum=10, Maximum=12)
     )
     Z=&Add_Float(
      A=0
      B=&FloatFromCurve(
       FloatCurve=DataInterface() {propertyValues {Curve {keys [
           {arriveTangent=0.1635, interpMode=RCIM_Cubic, leaveTangent=0.1635, tangentMode=RCTM_User, time=0, value=0}
           {arriveTangent=1.1312, interpMode=RCIM_Cubic, leaveTangent=1.1312, tangentMode=RCTM_User, time=1, value=1}
          ]}}}
       CurveIndex=&Divide_Float(A=$Particles.Age, B=2)
       Scale Curve=&RandomRangeFloat(Minimum=80, Maximum=80)
      )
     )
    )) {isSetParameters=true, module=SetVariables_#6}
   /SetVariables(Particles.Position=&AddVectorToPosition(Position=$Output.SocketLocation.SampledPosition, Vector=$Particles.SpiralVector)) {isSetParameters=true, module=SetVariables_#7}
   &CameraOffset(Camera Offset Amount=&Add_Float(A=5, B=&Multiply_Float(A=$Particles.Age, B=10))) {enabled=false, module=CameraOffset001}
   &UpdateMeshOrientation(
    Facing Direction=&NormalizeVector(
     Vector To Normalize=&Multiply_Vector(
      A=$Particles.SpiralVector
      B=&MakeVector(
       X=1
       Y=1
       Z=&FloatFromCurve(
        FloatCurve=DataInterface() {propertyValues {Curve {keys [<at 0 1>,<at 1 1>]}}}
        Scale Curve=&RandomRangeFloat(Minimum=1, Maximum=1)
       )
      )
     )
     Safe Normalize=true
     Safe Fallback Threshold=0
     Safe Fallback Vector=<0 0 0.001>
    )
    Facing Interpolation=false
    Orientation Method=::NewEnumerator1
    Override Quat=false
    TransformFacing=false
    TransformUp=false
   )
  ]
  rendererClasses [&NiagaraMeshRendererProperties]
  renderers [{rendererClass=&NiagaraMeshRendererProperties, rendererIndex=0}]
 }
 Flare {
  emitterUpdateScript [
   &EmitterState(Life Cycle Mode=::NewEnumerator1, Loop Behavior=::NewEnumerator1, Loop Delay=0.4, Loop Duration=0.3, UseLoopDelay=true)
   &SpawnRate(SpawnRate=50, Use Spawn Probability=false)
   &SpawnBurst_Instantaneous(Spawn Count=2, Use Spawn Probability=false, UseLoopCountLimit=false)
  ]
  particleSpawnScript [
   &InitializeParticle(
    Color=$System.TOD
    Lifetime=&RandomRangeFloat(Minimum=0.3, Maximum=0.4)
    Sprite Rotation Angle=45
    Sprite Rotation Mode=::NewEnumerator0
    Sprite Size=&Multiply_Vector2D_ByFloat(Float=&RandomRangeFloat(Minimum=20, Maximum=40), Vector2D=<1.1 1>)
    Sprite Size Mode=::NewEnumerator4
    UsePositionOffset=true
    Write Material Random=true
   )
   &CameraOffset(Camera Offset Amount=-20) {module=CameraOffset001}
   &PartitionParticles(Write Partition to Ribbon ID=::NewEnumerator0)
   /SetVariables(Particles.InitialEmitterAge:NiagaraFloat=$Emitter.Age) {isSetParameters=true}
  ]
  particleUpdateScript [
   &ParticleState()
   &SocketLocation(
    Interpolate Socket Sampling=true
    Orientation Sampling=::NewEnumerator1
    SamplePosOffset=true
    Sampled Position Offset=&MergeHelper_VectorSwitch(0=<6 0 0>, 1=<-6 0 0>, 2=?, 3=?, Index=$Particles.PartitionParticles.Partition, Num=::NewEnumerator0)
    Socket Index=$Particles.PartitionParticles.Partition
    Socket Reader=$System.Socket Reader
    Socket Sampling Mode=::NewEnumerator3
    VelocitySpeedLimit=false
   ) {module=SocketLocation001}
   &SkeletalMeshLocation(
    Bone / Socket Index=$Particles.PartitionParticles.Partition
    Bone Sampling Mode=::NewEnumerator4
    MaskSpawnGroup=false
    Orientation Sampling=::NewEnumerator1
    Position Sampling=::NewEnumerator2
    Sampled Position Offset=&MergeHelper_VectorSwitch(0=<6 0 0>, 1=<-6 0 0>, 2=?, 3=?, Index=$Particles.PartitionParticles.Partition, Num=::NewEnumerator0)
    Skeletal Mesh=$System.Skeletal Mesh
   ) {enabled=false}
   /SetVariables(Particles.Position=&AddVectorToPosition(
     Position=$Output.SocketLocation001.SampledPosition
     Vector=&Add_Vector(
      A=&Add_Vector(
       A=&Multiply_VectorByFloat(
        Vector=<1 0 0>
        Float=&Waveform(
         Global Phase=&Divide_Float(
          A=&MakeFloatFromInt(Int=&ReturnParticleIDAttributes())
          B=&MergeHelper_FloatSwitch(0=-15, 1=15, 2=?, 3=?, Index=$Particles.PartitionParticles.Partition, Num=::NewEnumerator0)
         )
         "Waveform [1]"=::NewEnumerator0
         "[1] Frequency"=1
         "[1] Amplitude Min/Max"=<-1 1>
         "[1] Scale"=&RandomRangeFloat(Minimum=10, Maximum=12)
        )
       )
       B=&Multiply_VectorByFloat(
        Vector=<0 1 0>
        Float=&Waveform(
         Global Phase=&Divide_Float(A=&MakeFloatFromInt(Int=&ReturnParticleIDAttributes()), B=15)
         "Waveform [1]"=::NewEnumerator1
         "[1] Frequency"=1
         "[1] Amplitude Min/Max"=<-1 1>
         "[1] Scale"=&RandomRangeFloat(Minimum=10, Maximum=12)
        )
       )
      )
      B=&MakeVector(
       X=0
       Y=0
       Z=&FloatFromCurve(
        FloatCurve=DataInterface() {propertyValues {Curve {keys [{tangentMode=RCTM_User, time=0, value=0}, {tangentMode=RCTM_User, time=0.3, value=1}]}}}
        CurveIndex=&Subtract_Float(A=$Particles.InitialEmitterAge, B=0.4)
        Scale Curve=20
       )
      )
     )
    )) {isSetParameters=true}
   /SetVariables(Particles.Position=&AddVectorToPosition(
     Position=$Output.SocketLocation001.SampledPosition
     Vector=&Add_Vector(
      A=&Multiply_VectorByFloat(
       Vector=&ReturnCameraVector(Camera Vectors=::NewEnumerator2, Coordinate Space=::Simulation)
       Float=&Waveform(
        Global Phase=&Divide_Float(
         A=&MakeFloatFromInt(Int=&ReturnParticleIDAttributes())
         B=&MergeHelper_FloatSwitch(0=-15, 1=15, 2=?, 3=?, Index=$Particles.PartitionParticles.Partition, Num=::NewEnumerator0)
        )
        "Waveform [1]"=::NewEnumerator0
        "[1] Frequency"=1
        "[1] Amplitude Min/Max"=<-1 1>
        "[1] Scale"=&RandomRangeFloat(Minimum=10, Maximum=12)
       )
      )
      B=&Multiply_VectorByFloat(
       Vector=&ReturnCameraVector(Camera Vectors=::NewEnumerator3, Coordinate Space=::Simulation)
       Float=&Waveform(
        Global Phase=&Divide_Float(A=&MakeFloatFromInt(Int=&ReturnParticleIDAttributes()), B=15)
        "Waveform [1]"=::NewEnumerator1
        "[1] Frequency"=1
        "[1] Amplitude Min/Max"=<-1 1>
        "[1] Scale"=&RandomRangeFloat(Minimum=10, Maximum=12)
       )
      )
     )
    )) {enabled=false, isSetParameters=true}
   &ScaleSpriteSize(Uniform Curve Sprite Scale=DataInterface() {propertyValues {Curve {keys [
        {arriveTangent=0.9776, interpMode=RCIM_Cubic, leaveTangent=0.9776, tangentMode=RCTM_User, time=0, value=0}
        <at 0.3 1 RCIM_Cubic>
        {arriveTangent=-0.8026, interpMode=RCIM_Cubic, leaveTangent=-0.8026, tangentMode=RCTM_User, time=1, value=0}
       ]}}})
   &SpriteRotationRate(Rotation Rate=&FloatFromCurve(FloatCurve=DataInterface() {propertyValues {Curve {keys [<at 0.2 1 RCIM_Cubic>,<at 0.4 180 RCIM_Cubic>]}}}))
   &ScaleColor(
    Scale Alpha=&FloatFromCurve(FloatCurve=DataInterface() {propertyValues {Curve {keys [<at 0.6 5>,<at 1 2>]}}})
    Scale RGB=&MakeVectorFromLinearColorRGB(Color=&MergeHelper_ColorSwitch(
      0=&ColorFromCurve(ColorCurve=DataInterface() {propertyValues {
         AlphaCurve {keys [<at 0 1 RCIM_Cubic>,<at 1 1 RCIM_Cubic>]}
         BlueCurve {keys [<at 0.0818 5>,<at 0.4288 0 RCIM_Cubic>]}
         GreenCurve {keys [<at 0.0818 4.7557>,<at 0.4288 0 RCIM_Cubic>]}
         RedCurve {keys [<at 0.0818 3.1251>,<at 0.4288 0 RCIM_Cubic>]}
        }})
      1=&ColorFromCurve(ColorCurve=DataInterface() {propertyValues {
         AlphaCurve {keys [<at 0 1 RCIM_Cubic>,<at 1 1 RCIM_Cubic>]}
         BlueCurve {keys [<at 0.0394 0.9982>,<at 0.2509 0.2093>,<at 0.6029 0.0837>,<at 0.9852 0.0219>]}
         GreenCurve {keys [<at 0.0394 1.2586>,<at 0.2509 0>,<at 0.6029 0>,<at 0.9852 0>]}
         RedCurve {keys [<at 0.0394 5>,<at 0.2509 5>,<at 0.6029 0.5417>,<at 0.9852 0.151>]}
        }})
      2=&ColorFromCurve(ColorCurve=DataInterface() {propertyValues {
         AlphaCurve {keys [<at 0 1 RCIM_Cubic>,<at 1 1 RCIM_Cubic>]}
         BlueCurve {keys [<at 0.0283 5>,<at 0.4401 0 RCIM_Cubic>]}
         GreenCurve {keys [<at 0.0283 4.7557>,<at 0.4401 0 RCIM_Cubic>]}
         RedCurve {keys [<at 0.0283 3.1251>,<at 0.4401 0 RCIM_Cubic>]}
        }})
      3=?
      Index=&ModuloInt(
       A=&Divide_Int(A:NiagaraInt32=&ReturnParticleIDAttributes(), B:NiagaraInt32=2)
       B=2
      )
      Num=::NewEnumerator0
     ))
    ScaleA=true
    ScaleRGB=true
   )
  ]
  rendererClasses [&NiagaraSpriteRendererProperties]
  renderers [{rendererClass=&NiagaraSpriteRendererProperties}]
 }
 Long_Cards {
  emitterUpdateScript [
   &EmitterState(
    Delay First Loop Only=true
    Inactive Response=::NewEnumerator1
    Life Cycle Mode=::NewEnumerator1
    Loop Delay=0.8
    Loop Duration=1
    UseLoopDelay=true
   )
   &SpawnRate(Spawn Probability=0.5, SpawnRate=50, Use Spawn Probability=true)
  ]
  particleSpawnScript [
   &PartitionParticles(Write Partition to Ribbon ID=::NewEnumerator0)
   &InitializeParticle(
    Color=<rgba 3 1.5567 1.5 1>
    Lifetime=&RandomRangeFloat(Minimum=1, Maximum=2)
    Mesh Renderer Array Visibility Mode=::NewEnumerator1
    Mesh Renderer Info=DataInterface() {propertyValues {MeshRenderer=&self:Long_Cards_0.NiagaraMeshRendererProperties_0}}
    Mesh Renderer Mesh Index:NiagaraInt32=&Divide_Int(A:NiagaraInt32=&ReturnParticleIDAttributes(), B:NiagaraInt32=2)
    Mesh Scale Mode=::NewEnumerator3
    Mesh Uniform Scale=0.2
    Sprite Size=&SelectVector2DFromArray(
     Direct Array Index=$Particles.PartitionParticles.Partition
     Vector 2D Selection Array=DataInterface() {propertyValues {FloatData [<60 230>, <60 230>, <30 100>, <30 100>]}}
    )
    Sprite Size Mode=::NewEnumerator4
    UsePositionOffset=false
    Write Material Random=true
   )
   &SocketLocation(
    Interpolate Socket Sampling=true
    Orientation Sampling=::NewEnumerator1
    SamplePosOffset=true
    Sampled Position Offset=&FN_FrontEndVector3Modifier(
     Gameplay=&MergeHelper_VectorSwitch(0=<3 0 0>, 1=<-3 0 0>, 2=?, 3=?, Index=$Particles.PartitionParticles.Partition, Num=::NewEnumerator0)
     Front End=<0 0 0>
    )
    Socket Index=$Particles.PartitionParticles.Partition
    Socket Reader=$System.Socket Reader
    Socket Sampling Mode=::NewEnumerator3
    VelocitySpeedLimit=false
   ) {module=SocketLocation001}
   &InitialMeshOrientation(Mesh Orientation Mode=::NewEnumerator7, Use Rotation Vector=false)
   &CameraOffset(Camera Offset Amount=-10)
  ]
  particleUpdateScript [
   &ParticleState()
   &ScaleMeshSize(Scale Factor=&Multiply_Vector(
     A=&Lerp_Vector(
      A=<1 1 0.1>
      B=<0.8 1 10>
      Alpha=&Waveform(
       Waveform Count=::NewEnumerator1
       Global Input Position=&RandomRangeFloat(Minimum=0, Maximum=100)
       Global Amplitude Scale=&FloatFromCurve(FloatCurve=DataInterface() {propertyValues {Curve {keys [<at 0.3 0>,<at 0.5 1>]}}})
       "Waveform [1]"=::NewEnumerator0
       "[1] Frequency"=4
       "[1] Amplitude Min/Max"=<0 1>
       "Waveform [2] Blend"=::NewEnumerator2
       "[2] Input Position"=0
       "[2] Phase"=$Emitter.Age
       "[2] Phase Scale"=1
       "[2] Compound Frequency"=2
       "[2] Frequency"=0.3
       "[2] Amplitude Min/Max"=<0 1>
       "[2] Scale"=1
       "[2] Bias"=0
       "[2] Exponent"=1
       "[2] Pulse Duration"=0.1
       "[2] Clip Triangle Min/Max"=<0 0>
       "[2] Random Seed"=0
      )
     )
     B=&VectorFromFloat(Value=&Multiply_Float(
       A=&FloatFromCurve(FloatCurve=DataInterface() {propertyValues {Curve {keys [
            <at 0.2 0 RCIM_Cubic>
            {arriveTangent=6.3053, leaveTangent=1.1111, tangentMode=RCTM_User, time=0.4, value=1}
            {arriveTangent=-1.4286, leaveTangent=-1.4286, time=0.9, value=1}
            <at 1 0>
           ]}}})
       B=/Niagara/DynamicInputs/Camera/CameraFade(
        Camera Query:NiagaraDataInterfaceCamera=DataInterface()
        Fade Offset:NiagaraFloat=30
        Fade Distance:NiagaraFloat=100
        Use Owner Position:NiagaraBool=true
       )
      ))
    ))
   &SocketLocation(
    Interpolate Socket Sampling=true
    Orientation Sampling=::NewEnumerator1
    SamplePosOffset=true
    Sampled Position Offset=&MergeHelper_VectorSwitch(0=<3 0 0>, 1=<-3 0 0>, 2=?, 3=?, Index=$Particles.PartitionParticles.Partition, Num=::NewEnumerator0)
    Socket Index=$Particles.PartitionParticles.Partition
    Socket Reader=$System.Socket Reader
    Socket Sampling Mode=::NewEnumerator3
    VelocitySpeedLimit=false
   )
   &SkeletalMeshLocation(
    Bone / Socket Index=$Particles.PartitionParticles.Partition
    Bone Sampling Mode=::NewEnumerator4
    MaskSpawnGroup=false
    Orientation Sampling=::NewEnumerator1
    Position Sampling=::NewEnumerator2
    Sampled Position Offset=&MergeHelper_VectorSwitch(0=<3 0 0>, 1=<-3 0 0>, 2=?, 3=?, Index=$Particles.PartitionParticles.Partition, Num=::NewEnumerator0)
    Skeletal Mesh=$System.Skeletal Mesh
   ) {enabled=false}
   /SetVariables(Particles.CustomRandom:NiagaraFloat=&Add_Float(
     A=&RandomRangeFloat(
      Minimum=0
      Maximum=50
      Randomness Mode=::NewEnumerator1
      Random Seed=&Add_Integer(A=$Particles.PartitionParticles.Partition, B=0)
      Fixed Random Seed=true
     )
     B=&Multiply_Float(
      A=$Emitter.Age
      B=&RandomRangeFloat(Minimum=0.1, Maximum=0.2, Randomness Mode=::NewEnumerator1, Random Seed=$Particles.PartitionParticles.Partition, Fixed Random Seed=true)
     )
    )) {isSetParameters=true, module=SetVariables_#5}
   /SetVariables(Particles.SpiralVector:Vector3f=&MakeVector(
     X=&Waveform(
      Global Input Position=&Add_Float(
       A=$Particles.CustomRandom
       B=&FloatFromCurve(
        FloatCurve=DataInterface() {propertyValues {Curve {keys [
            {arriveTangent=3.2073, interpMode=RCIM_Cubic, leaveTangent=3.2073, tangentMode=RCTM_User, time=0, value=0}
            {arriveTangent=1, interpMode=RCIM_Cubic, leaveTangent=1, time=0.3, value=0.7}
            <at 1 1 RCIM_Cubic>
           ]}}}
        CurveIndex=&Multiply_Float(A=$Particles.Age, B=0.5)
        Scale Curve=&Multiply_Float(
         A=5
         B=&RandomRangeFloat(
          Minimum=0.7
          Maximum=1
          Randomness Mode=::NewEnumerator1
          Random Seed=&Add_Integer(A=$Particles.PartitionParticles.Partition, B=44853)
          Fixed Random Seed=true
         )
        )
       )
      )
      Global Amplitude Scale=&MergeHelper_FloatSwitch(0=-1, 1=1, 2=?, 3=?, Index=$Particles.PartitionParticles.Partition, Num=::NewEnumerator0)
      "Waveform [1]"=::NewEnumerator0
      "[1] Phase"=$Particles.NormalizedAge
      "[1] Frequency"=0.5
      "[1] Amplitude Min/Max"=<-1 1>
      "[1] Scale"=&FloatFromCurve(
       FloatCurve=DataInterface() {propertyValues {Curve {keys [<at 0 5>,<at 1 30>]}}}
       CurveIndex=&Multiply_Float(A=$Particles.Age, B=0.5)
      )
     )
     Y=&Waveform(
      Global Input Position=&Add_Float(
       A=$Particles.CustomRandom
       B=&FloatFromCurve(
        FloatCurve=DataInterface() {propertyValues {Curve {keys [
            {arriveTangent=3.2073, interpMode=RCIM_Cubic, leaveTangent=3.2073, tangentMode=RCTM_User, time=0, value=0}
            {arriveTangent=1, interpMode=RCIM_Cubic, leaveTangent=1, time=0.3, value=0.7}
            <at 1 1 RCIM_Cubic>
           ]}}}
        CurveIndex=&Multiply_Float(A=$Particles.Age, B=0.5)
        Scale Curve=&Multiply_Float(
         A=5
         B=&RandomRangeFloat(
          Minimum=0.7
          Maximum=1
          Randomness Mode=::NewEnumerator1
          Random Seed=&Add_Integer(A=$Particles.PartitionParticles.Partition, B=44853)
          Fixed Random Seed=true
         )
        )
       )
      )
      "Waveform [1]"=::NewEnumerator1
      "[1] Phase"=$Particles.NormalizedAge
      "[1] Frequency"=0.5
      "[1] Amplitude Min/Max"=<-1 1>
      "[1] Scale"=&FloatFromCurve(
       FloatCurve=DataInterface() {propertyValues {Curve {keys [<at 0 5>,<at 1 30>]}}}
       CurveIndex=&Multiply_Float(A=$Particles.Age, B=0.5)
      )
     )
     Z=&Add_Float(
      A=5
      B=&FloatFromCurve(
       FloatCurve=DataInterface() {propertyValues {Curve {keys [
           {arriveTangent=0.5827, interpMode=RCIM_Cubic, time=0, value=0}
           {arriveTangent=1.8532, interpMode=RCIM_Cubic, leaveTangent=1.8532, tangentMode=RCTM_User, time=1, value=1}
          ]}}}
       CurveIndex=&Multiply_Float(A=$Particles.Age, B=0.5)
       Scale Curve=300
      )
     )
    )) {isSetParameters=true, module=SetVariables_#6}
   /SetVariables(Particles.Position=&AddVectorToPosition(
     Position=&AddVectorToPosition(
      Position=&Multiply_Vector(
       A=&Lerp_Vector(
        A=Linked() {linkedVariable {name=Particles.Initial.Position, type=NiagaraPosition}}
        B=$Output.SocketLocation.SampledPosition
        Alpha=&FloatFromCurve(FloatCurve=DataInterface() {propertyValues {Curve {keys [<at 0.55 1 RCIM_Cubic>, {arriveTangent=-0.5694, interpMode=RCIM_Cubic, leaveTangent=-0.5694, tangentMode=RCTM_User, time=1, value=0.9}]}}})
       )
       B=<1 1 0>
      )
      Vector=&Multiply_Vector(A=$Output.SocketLocation.SampledPosition, B=<0 0 1>)
     )
     Vector=&FN_FrontEndVector3Modifier(
      Gameplay=$Particles.SpiralVector
      Front End=&RotateVectorByQuaternion(
       Vector To Rotate=$Particles.SpiralVector
       Quaternion=&MakeQuatFromAxisAngle(
        Axis=/Game/Effects/Niagara/DynamicInputs/Engine/FN_GetSystemAxis(Axis:ENiagaraOrientationAxis=::XAxis, Flip Axis:NiagaraBool=false)
        AngleInDegrees=30
       )
      )
     )
    )) {isSetParameters=true, module=SetVariables_#7}
   &UpdateMeshOrientation(
    Facing Direction=&NormalizeVector(
     Vector To Normalize=&Multiply_Vector(A=$Particles.SpiralVector, B=<1 1 0.5>)
     Safe Normalize=true
     Safe Fallback Threshold=0
     Safe Fallback Vector=<0 0 0.001>
    )
    Facing Interpolation=false
    Orientation Method=::NewEnumerator1
    Override Quat=false
    TransformFacing=false
    TransformUp=false
   )
   &UpdateMeshOrientation(
    Rotation Rate=&FloatFromCurve(FloatCurve=DataInterface() {propertyValues {Curve {keys [<at 0.3 0>,<at 0.5 1>]}}})
    Rotation Vector=&RandomRangeVector(Minimum=<-1 -1 -1>, Maximum=<1 1 1>)
   ) {module=UpdateMeshOrientation001}
  ]
  rendererClasses [&NiagaraMeshRendererProperties]
  renderers [{rendererClass=&NiagaraMeshRendererProperties, rendererIndex=0}]
 }
 Speed_Lines {
  emitterUpdateScript [
   &EmitterState(Delay First Loop Only=true, Life Cycle Mode=::NewEnumerator1, Loop Delay=1.5, Loop Duration=1, UseLoopDelay=true)
   &SpawnRate(
    SpawnRate=&FN_FrontEndFloatModifier(Gameplay=&Multiply_Float(A=50, B=1), Front End=50)
    Use Spawn Probability=false
   ) {module=SpawnRate001}
  ]
  particleSpawnScript [
   &InitializeParticle(
    Color=&MultiplyLinearColor(
     Linear Color=&MultiplyLinearColor(
      Linear Color=&MergeHelper_ColorSwitch(
       0=<rgba 0 0 0 3>
       1=<rgba 1 0 0.0615 1>
       2=<rgba 0.733 0.9995 1 1>
       3=?
       Index=&RandomRangeInt(Minimum=0, Maximum=2)
       Num=::NewEnumerator1
      )
      Scale Factor=&MakeVector4_FromLinearColor(LinearColor=$System.TOD)
     )
     Scale Factor=&MakeVector4(X=1, Y=1, Z=1, W=&FN_FrontEndFloatModifier(Gameplay=0.12, Front End=0.1))
    )
    Lifetime Max=0.25
    Lifetime Min=0.15
    Lifetime Mode=::NewEnumerator1
    Sprite Size Max=<3 7>
    Sprite Size Min=<2 4>
    Sprite Size Mode=::NewEnumerator2
    UsePositionOffset=true
    Write Material Random=true
   )
   &ShapeLocation(
    Cylinder Radius=&FN_FrontEndFloatModifier(Gameplay=100, Front End=100)
    MaskSpawnGroup=false
    Offset=&FN_FrontEndVector3Modifier(Gameplay=<0 0 50>, Front End=<0 50 -30>)
    Offset Mode=::NewEnumerator0
    Override Seed=false
    Shape Primitive=::NewEnumerator1
    Surface Only=false
    UseOwnerScale=false
   )
   &AddVelocity(
    Rotation Coordinate Space=::World
    Velocity=&FN_FrontEndVector3Modifier(
     Gameplay=&Add_Vector(
      A=&RandomRangeVector(Minimum=<0 0 1000>, Maximum=<0 0 1200>)
      B=&Multiply_Vector(A=$Engine.Owner.Velocity, B=&RandomRangeVector(Minimum=<-0.2 -0.2 0>, Maximum=<-0.5 -0.5 0>))
     )
     Front End=&TransformVector(Vector=&RandomRangeVector(Minimum=<0 -500 1000>, Maximum=<0 -600 1000>), Destination Space=::World)
    )
   )
  ]
  particleUpdateScript [
   &ParticleState()
   &ScaleSpriteSize(Scale Factor=&Vector2DFromCurve(
     Vector2Curve=DataInterface() {propertyValues {XCurve {keys [<at 0 0.5 RCIM_Cubic>,<at 0.5 1 RCIM_Cubic>]}, YCurve {keys [<at 0 0 RCIM_Cubic>,<at 0.5 1 RCIM_Cubic>]}}}
     ScaleCurve=&Vector2DFromFloat(Value=1)
    ))
   &SolveForcesAndVelocity(Clamp Velocity=false, Limit Acceleration=false)
   &ScaleSpriteSizeBySpeed(Max Scale Factor=<1 30>, Min Scale Factor=<0.01 0.05>, Velocity Threshold=1)
   &ScaleColor(
    Scale Alpha=&FloatFromCurve(
     FloatCurve=DataInterface() {propertyValues {Curve {keys [
         {arriveTangent=6.8397, interpMode=RCIM_Cubic, leaveTangent=6.8397, tangentMode=RCTM_User, time=0, value=0}
         <at 0.3 1 RCIM_Cubic>
         {arriveTangent=-2.221, interpMode=RCIM_Cubic, leaveTangent=-2.221, tangentMode=RCTM_User, time=1, value=0}
        ]}}}
     Scale Curve=&FN_FrontEndFloatModifier(
      Gameplay=&Multiply_Float(
       A=$System.Dive
       B=&Lerp_Float(
        A=0
        B=&FNTimeOfDayScalar(Night Scale=0.4, Morning Scale=0.5, Day Scale=0.6, Evening Scale=0.4)
        Alpha=1
       )
      )
      Front End=&Lerp_Float(A=0, B=0.4, Alpha=$System.SystemFadeIn)
     )
    )
    Scale RGB=&VectorFromFloat(Value=&FN_FrontEndFloatModifier(Gameplay=1, Front End=1))
    ScaleA=true
    ScaleRGB=true
   )
  ]
  rendererClasses [&NiagaraSpriteRendererProperties]
  renderers [{rendererClass=&NiagaraSpriteRendererProperties}]
 }
}
summary {
 systemName=NS_Contrail_CasinoReaper
 emitters [
  {emitterName=Long_Cards, rendererClasses [&NiagaraMeshRendererProperties]}
  {emitterName=Card_Base, rendererClasses [&NiagaraMeshRendererProperties]}
  {emitterName=Flare, rendererClasses [&NiagaraSpriteRendererProperties]}
  {emitterName=Speed_Lines, rendererClasses [&NiagaraSpriteRendererProperties]}
 ]
 userVariables [
  User.bIsFrontEnd {type=NiagaraBool}
  User.bIsFrontEndPreview {type=NiagaraBool}
  User.ParaGlideLeanAlpha {type=NiagaraFloat, defaultValue=0, description=""}
 ]
}
